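# Tab button for the settings top nav bar, built on the selectable UIKit button, with a material-backed background and a title label.
# Selection changes share one branch that shows or collapses the label, snaps modular blocks to selected or deselected state, and sets text background opacity.
# Hover plays a background fade forward and in reverse, only while unselected.
# PreConstruct layers the main slot and applies mobile material and font tweaks; SetTabTitleText updates the label, preserving visibility.
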WidgetBlueprint /FortniteCoreUI/Settings/Tabs/WBP_Settings_Button_Tab_TopNavBar
animations [
 Anim_IsSelected {
  binding_count=3
  length_seconds=0.4
  object=&self:Anim_IsSelected
  sequence {
   bindings [
    CommonVisualAttachment_107 {
     id=#1
     possessed_class=CommonVisualAttachment
     tracks [{
       type=MovieSceneFloatTrack
       display_name=Height Override
       sections [{
         type=MovieSceneFloatSection
         end_frame=24
         start_frame=0
         channels [None {type=MovieSceneScriptingFloatChannel, keys [[0,216],[3,256],[24,216]]}]
        }]
      }]
    }
    BackgroundVisuals {
     id=#2
     possessed_class=Image
     tracks [{
       type=MovieSceneWidgetMaterialTrack
       sections [{
         type=MovieSceneParameterSection
         channels [
          Flare Scale X {type=MovieSceneScriptingFloatChannel, keys [[0,0.5],[13.068,0.2]]}
          Flare Opacity {type=MovieSceneScriptingFloatChannel, keys [[0,1],[13.068,0]]}
          Triangle Remover {type=MovieSceneScriptingFloatChannel, thinned_from=4, keys [[0,1],[10.454,1],[18.295,0]]}
          Global Opacity {type=MovieSceneScriptingFloatChannel, keys [[0,0],[1.307,1],[18.295,0]]}
         ]
        }]
      }]
    }
    TextBackground {
     id=#3
     possessed_class=Image
     tracks [
      {
       type=MovieSceneWidgetMaterialTrack
       sections [{type=MovieSceneParameterSection, channels [Padding X {type=MovieSceneScriptingFloatChannel, keys [[0,0],[18,2]]}]}]
      }
      {
       type=MovieSceneFloatTrack
       display_name=Render Opacity
       sections [{type=MovieSceneFloatSection, channels [None {type=MovieSceneScriptingFloatChannel, const [1,1]}]}]
      }
     ]
    }
   ]
   display_rate {numerator=60}
  }
 }
 Anim_IsHovered {
  binding_count=1
  length_seconds=0.2
  object=&self:Anim_IsHovered
  sequence {
   bindings [TextBackground {
     id=#4
     possessed_class=Image
     tracks [{
       type=MovieSceneFloatTrack
       display_name=Render Opacity
       sections [{type=MovieSceneFloatSection, end_frame=12, start_frame=0, channels [None {type=MovieSceneScriptingFloatChannel, keys [[0,0],[12,0.2]]}]}]
      }]
    }]
   display_rate {numerator=60}
  }
 }
 Anim_UnSelected {
  binding_count=1
  length_seconds=0.0167
  object=&self:Anim_UnSelected
  sequence {
   bindings [TextBackground {
     id=#3
     possessed_class=Image
     tracks [{
       type=MovieSceneFloatTrack
       display_name=Render Opacity
       sections [{type=MovieSceneFloatSection, channels [None {type=MovieSceneScriptingFloatChannel, const [0,1]}]}]
      }]
    }]
   display_rate {numerator=60}
  }
 }
]
graphs [
 EventGraph {nodes [
   n1 = Event PreConstruct()
   > Parent: PreConstruct(IsDesignTime=@n1.IsDesignTime)
   > SetGridSlotLayer!(execute, Widget:Widget Object Reference=Get Slot_Main() -> (Slot_Main:Named Slot Object Reference), Layer:int=5) -> (then)
   > SetMobileStyle() {is=title K2Node_CallFunction}
   Event Construct()
   > Parent: Construct()
   > n2 = GetDynamicMaterial(self=Get BackgroundVisuals() -> (BackgroundVisuals:Image Object Reference))
   > SetScalarParameterValue(self=@n2, ParameterName=Global Opacity, Value=0)
   > n3 = Branch(Condition=GetSelected!(self:Common Button Base Object Reference) -> (ReturnValue:bool))
   > RequestInstantTransitionOnBlocks!(execute, self:UIKit Modular Button Object Reference, State:EUIKitBlockInstantTransitionState Enum=InstantSelected) -> (then)
   > SetRenderOpacity(self=Get TextBackground() -> (TextBackground:Image Object Reference), InOpacity=1)
   RequestInstantTransitionOnBlocks(execute=@n3.else, State=InstantDeselected)
   > SetRenderOpacity(self=Get TextBackground(), InOpacity=0)
   Event On Hovered()
   > Parent: On Hovered() {is=title K2Node_CallParentFunction}
   n4 = NOT Boolean(A=GetSelected())
   > Branch(Condition=@n4)
   > PlayAnimationForward(InAnimation=Get Anim_IsHovered() -> (Anim_IsHovered:Widget Animation Object Reference))
   Event On Unhovered()
   > Parent: On Unhovered() {is=title K2Node_CallParentFunction}
   > Branch(Condition=@n4)
   > PlayAnimationReverse(InAnimation=Get Anim_IsHovered())
   Event On Selected()
   > n5 = Parent: On Selected() {is=title K2Node_CallParentFunction}
   n6 = Branch(execute=[@n5.then, @n7.then], Condition=GetSelected())
   > SetVisibility(self=Get TextLabel() -> (TextLabel:WBP UIKit Block Text Object Reference), InVisibility=HitTestInvisible)
   > RequestInstantTransitionOnBlocks(State=InstantSelected)
   > SetRenderOpacity(self=Get TextBackground(), InOpacity=1)
   SetVisibility(execute=@n6.else, self=Get TextLabel(), InVisibility=Collapsed)
   > RequestInstantTransitionOnBlocks(State=InstantDeselected)
   > SetRenderOpacity(self=Get TextBackground(), InOpacity=0)
   Event On Deselected()
   > n7 = Parent: On Deselected() {is=title K2Node_CallParentFunction}
  ]}
 SetMobileStyle {nodes [
   SetMobileStyle() {is=title K2Node_FunctionEntry}
   > Branch(Condition=IsMobileGame())
   > SetBrushFromMaterial(self=Get TextBackground(), Material=/FortniteCoreUI/UI/Frontend/TopBar/MI_UIKit_Tab_Main_ContentBackground_Mobile)
   n1 = Get TextLabel()
   > SetGridSlotNudge!(execute, Widget:Widget Object Reference=@n1, Nudge:Vector 2D Structure=(X=0,Y=42)) -> (then)
   > SetMobileFontSizeMultiplier!(execute, self:WBP UIKit Block Text Object Reference=@n1, MobileFontSizeMultiplier:double=1.3) -> (then)
   > SetFont!(
    execute
    self:WBP UIKit Block Text Object Reference=@n1
    Font_FontObject:Object Reference=&HeadingNow
    Font_FontMaterial:Object Reference
    Font_OutlineSettings:Font Outline Settings Structure
    Font_TypefaceFontName:Name=76 Bold
    Font_Size:float=16
    Font_LetterSpacing:int=10
    Font_SkewAmount:float=0
    Font_bForceMonospaced:bool=false
    Font_bMaterialIsStencil:bool=false
    Font_MonospacedWidth:float=1
   ) -> (then)
  ]}
 SetTabTitleText {nodes [
   n1 = SetTabTitleText!() -> (then, InTabTitleText:Text) {is=title K2Node_FunctionEntry}
   > n2 = Set TextPageTitle(TextPageTitle:Text=@n1) -> (Output_Get:Text)
   > Set PreviousVisibility(PreviousVisibility:ESlateVisibility Enum=GetVisibility!(self:Widget Object Reference=Get TextLabel()) -> (ReturnValue:ESlateVisibility Enum=Visible)) -> (Output_Get:ESlateVisibility Enum=Visible)
   n3 = Get TextLabel()
   > SetText!(execute, self:WBP UIKit Block Text Object Reference=@n3, Text:Text=@n2) -> (then)
   > SetVisibility(self=@n3, InVisibility=Get PreviousVisibility() -> (PreviousVisibility:ESlateVisibility Enum=Visible))
  ]}
]
variables [
 FontOutline {type=number}
 GridPadding {type=Margin, default=<ltrb 0 0 4 0>}
 TextPageTitle {type=string, default=""}
]
widgets {
 info {
  inheritedWidgetCount=0
  namedSlotCount=1
  parentClass=/FortUILibrary/UIKit/Buttons/WBP_UIKit_Button_Selectable.WBP_UIKit_Button_Selectable_C
  rootWidgetClass=null
  widgetCount=6
 }
 widgets [
  GridPanel_50 {widgetClassPath=&GridPanel, widget=GridPanel_50, properties {columnFill [1], rowFill [1]}}
  CommonVisualAttachment_107 {
   widgetClassPath=&CommonVisualAttachment
   parent=GridPanel_50
   widget=CommonVisualAttachment_107
   properties {
    bOverride_HeightOverride=true
    bOverride_WidthOverride=true
    heightOverride=216
    visibility=HitTestInvisible
    widthOverride=192
    contentAnchor=<0.5 0>
    renderTransform {angle=0, scale=<1 1>, shear=<0 0>, translation=<0 -40>}
   }
   slot {class=GridSlot, properties {horizontalAlignment=HAlign_Center, verticalAlignment=VAlign_Top}}
  }
  BackgroundVisuals {
   widgetClassPath=&Image
   bIsVariable=true
   parent=CommonVisualAttachment_107
   widget=BackgroundVisuals
   properties {
    visibility=Collapsed
    brush {
     bIsDynamicallyLoaded=false
     drawAs=Image
     imageType=NoImage
     mirroring=NoMirror
     resourceName=None
     resourceObject=/Game/UI/Frontend/TopBar/MI_UIKit_Tab_Main_Background
     tiling=NoTile
     imageSize=<32 32>
     margin=<ltrb 0 0 0 0>
     outlineSettings {
      bUseBrushTransparency=false
      roundingType=HalfHeightRadius
      width=0
      color {colorUseRule=UseColor_Specified, specifiedColor=<rgba 0 0 0 0>}
      cornerRadii=<0 0 0 0>
     }
     tintColor {colorUseRule=UseColor_Specified, specifiedColor=<rgba 1 1 1 1>}
     uVRegion {bIsValid=false, max=<0 0>, min=<0 0>}
    }
    renderTransformPivot=<0.5 0>
   }
   slot {class=SizeBoxSlot}
  }
  SizeBox {widgetClassPath=&SizeBox, bIsVariable=true, parent=GridPanel_50, widget=SizeBox, slot {class=GridSlot, properties {layer=1}}}
  TextBackground {
   widgetClassPath=&Image
   bIsVariable=true
   parent=SizeBox
   widget=TextBackground
   properties {
    visibility=SelfHitTestInvisible
    brush {
     bIsDynamicallyLoaded=false
     drawAs=Image
     imageType=NoImage
     mirroring=NoMirror
     resourceName=None
     resourceObject=/Game/UI/Frontend/TopBar/MI_UIKit_Tab_Main_ContentBackground
     tiling=NoTile
     imageSize=<32 32>
     margin=<ltrb 0 0 0 0>
     outlineSettings {
      bUseBrushTransparency=false
      roundingType=HalfHeightRadius
      width=0
      color {colorUseRule=UseColor_Specified, specifiedColor=<rgba 0 0 0 0>}
      cornerRadii=<0 0 0 0>
     }
     tintColor {colorUseRule=UseColor_Specified, specifiedColor=<rgba 1 1 1 1>}
     uVRegion {bIsValid=false, max=<0 0>, min=<0 0>}
    }
    colorAndOpacity=<rgba 0.7913 0.7835 0.7835 1>
   }
   slot {class=SizeBoxSlot}
  }
  TextLabel {
   widgetClassPath=/FortUILibrary/Systems/ModularButton/Blocks/WBP_UIKit_Block_Text
   bIsVariable=true
   parent=GridPanel_50
   widget=TextLabel
   properties {visibility=Collapsed}
   slot {class=GridSlot, properties {verticalAlignment=VAlign_Bottom, nudge=<0 45>, padding=<ltrb -200 0 -200 0>}}
  }
 ]
}
